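# Niagara effect for the FloralCardinal backpack cosmetic, adding thruster flames and glows plus a glowing button.
# Thrusters places flickering, hue-shifted mesh flames at sockets read through a shared Socket Reader; ThrusterGlow adds two socket-oriented sprite layers.
# ButtonGlow renders a pulsing sprite and light at a fixed offset.
# Burst-spawned particles persist, scaled by System.FadeIn, while camera dot products shift glow position and drive thruster glow alpha, size and depth offset.
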NiagaraSystem /BRCosmetics/Accessories/FORT_Backpacks/Backpack_M_MED_FloralCardinal/FX/NS_Backpack_FloralCardinal
dependencies {
 usedDataInterfaces [
  &NiagaraDataInterfaceCamera
  &NiagaraDataInterfaceMeshRendererInfo
  &NiagaraDataInterfaceColorCurve
  &NiagaraDataInterfaceCurve
  &NiagaraDataInterfaceSocketReader
 ]
 usedDynamicInputs [&Multiply_VectorByFloat, &Subtract_Float]
 usedModules [
  &self:SystemSpawnScript.SystemScriptSource.SystemScriptGraph.NiagaraNodeAssignment_0.SetVariables_#1
  &self:SystemSpawnScript.SystemScriptSource.SystemScriptGraph.NiagaraNodeAssignment_4.SetVariables_#2
  &SystemState
  &self:SystemSpawnScript.SystemScriptSource.SystemScriptGraph.NiagaraNodeAssignment_1.SetVariables_#3
  &self:SystemSpawnScript.SystemScriptSource.SystemScriptGraph.NiagaraNodeAssignment_7.SetVariables_#4
 ]
}
emitters {
 ButtonGlow {
  emitterUpdateScript [
   &EmitterState(Life Cycle Mode=::NewEnumerator1, Loop Behavior=::NewEnumerator1, Loop Duration=9.4167, UseLoopDelay=false)
   &SpawnBurst_Instantaneous(Use Spawn Probability=false, UseLoopCountLimit=false)
  ]
  particleSpawnScript [
   &PartitionParticles()
   &InitializeParticle(
    Color=<rgba 0 0.7775 1 0.5>
    Lifetime=2
    Mesh Scale Mode=::NewEnumerator3
    Mesh Uniform Scale=0.6
    Position Offset=<0 -18 2.5>
    Sprite Size Mode=::NewEnumerator3
    Uniform Sprite Size=15
    UsePositionOffset=true
    Write Material Random=true
   )
   &CameraOffset(Camera Offset Amount=5)
  ]
  particleUpdateScript [
   &ParticleState(Kill Particles When Lifetime Has Elapsed=false)
   &ScaleColor(
    Scale Alpha=&Multiply_Float(
     A=&Waveform(
      Global Input Position=&RandomRangeFloat(Minimum=0, Maximum=10)
      "Waveform [1]"=::NewEnumerator0
      "[1] Frequency"=1
      "[1] Amplitude Min/Max"=<1.4 2>
     )
     B=$System.FadeIn
    )
    ScaleA=true
    ScaleRGB=true
   )
   &SolveForcesAndVelocity(Clamp Velocity=false, Limit Acceleration=false)
   /SetVariables(Particles.Position=&AddVectorToPosition(
     Position=$Particles.Initial.Position
     Vector=&Lerp_Vector(
      A=<0 3 0>
      B=<0 1 0>
      Alpha=&FloatFromCurve(
       FloatCurve=DataInterface() {propertyValues {Curve {keys [<at 0 0 RCIM_Cubic>, {arriveTangent=2.305, interpMode=RCIM_Cubic, leaveTangent=2.305, tangentMode=RCTM_User, time=1, value=1}]}}}
       CurveIndex=&Abs_Float(Float:NiagaraFloat=&DotProduct(
         Vector A=&Multiply_Vector(A=&ReturnCameraVector(Camera Vectors=::NewEnumerator1, Coordinate Space=::Local), B=<1 1 1>)
         Vector B=<0 1 0>
        ))
      )
     )
    )) {isSetParameters=true}
   &FN_Light_Attributes(
    Color=<rgba 0 7.211 15 1>
    Create Color Attribute=true
    Create Exponent Attribute=true
    Create Position Attribute=true
    Create Radius Attribute=true
    Exponent=1
    Position=<0 -21 2.5>
    Radius=10
    Scale Color=&MakeFloatFromVector4(Vector 4=&MakeVector4_FromLinearColor(LinearColor=$Particles.Color), Channel=::NewEnumerator3)
   )
  ]
  rendererClasses [&NiagaraSpriteRendererProperties, &NiagaraLightRendererProperties]
  renderers [
   {rendererClass=&NiagaraSpriteRendererProperties}
   {rendererClass=&NiagaraLightRendererProperties, rendererIndex=1}
  ]
 }
 ThrusterGlow {
  emitterUpdateScript [
   &EmitterState(Life Cycle Mode=::NewEnumerator1, Loop Behavior=::NewEnumerator1, Loop Duration=9.4167, UseLoopDelay=false)
   &SpawnBurst_Instantaneous(Spawn Count=2, Use Spawn Probability=false, UseLoopCountLimit=false)
  ]
  particleSpawnScript [
   &PartitionParticles()
   &InitializeParticle(
    Color=<rgba 2 2 2 0.5>
    Lifetime=2
    Mesh Scale Mode=::NewEnumerator3
    Mesh Uniform Scale=0.6
    Sprite Size=&Multiply_Vector2D(A=<7 14>, B=<1 1>)
    Sprite Size Mode=::NewEnumerator4
    UsePositionOffset=false
    Write Material Random=true
   )
   &SocketLocation(
    Orientation Sampling=::NewEnumerator1
    SamplePosOffset=true
    Sampled Position Offset=<0 0 1>
    Socket Index=$Particles.PartitionParticles.Partition
    Socket Reader=$System.Socket Reader
    Socket Sampling Mode=::NewEnumerator3
    VelocitySpeedLimit=false
   ) {module=SocketLocation001}
   &AddVelocity(
    Velocity=&RotateVectorByQuaternion(Vector To Rotate=<0 0 0>, Quaternion=$Output.SocketLocation.SampledRotation)
    Velocity Speed Scale=50
   ) {enabled=false}
  ]
  particleUpdateScript [
   &ParticleState(Kill Particles When Lifetime Has Elapsed=false)
   &SocketLocation(
    Orientation Sampling=::NewEnumerator1
    SamplePosOffset=true
    Socket Index=$Particles.PartitionParticles.Partition
    Socket Reader=$System.Socket Reader
    Socket Sampling Mode=::NewEnumerator3
    VelocitySpeedLimit=false
   )
   /SetVariables(Particles.GlowDot:NiagaraFloat=&ClampFloat(Float=&DotProduct(
      Vector A=&ReturnCameraVector(Camera Vectors=::NewEnumerator4, Coordinate Space=::Simulation, Origin Position=$Particles.Position)
      Vector B=&RotateVectorByQuaternion(Vector To Rotate=<0 0 1>, Quaternion=$Output.SocketLocation.SampledRotation)
     ))) {isSetParameters=true}
   &CameraOffset(Camera Offset Amount=&Lerp_Float(A=-2, B=5, Alpha=$Particles.GlowDot))
   &ScaleColor(
    Scale Alpha=&Lerp_Float(
     A=0
     B=&Multiply_Float(
      A=&Waveform(
       Global Input Position=&RandomRangeFloat(Minimum=0, Maximum=10)
       "Waveform [1]"=::NewEnumerator0
       "[1] Frequency"=1
       "[1] Amplitude Min/Max"=<1.4 2>
      )
      B=$System.FadeIn
     )
     Alpha=$Particles.GlowDot
    )
    ScaleA=true
    ScaleRGB=true
   )
   &SolveForcesAndVelocity(Clamp Velocity=false, Limit Acceleration=false)
   /SetVariables(Particles.Position=&AddVectorToPosition(
     Position=$Particles.Position
     Vector=&Lerp_Vector(
      A=&RotateVectorByQuaternion(Vector To Rotate=<0 0 -1>, Quaternion=$Output.SocketLocation.SampledRotation)
      B=<0 0 0>
      Alpha=$Particles.GlowDot
     )
    )) {isSetParameters=true}
   &SpriteFacingAndAlignment(
    Sprite Alignment=&RotateVectorByQuaternion(Vector To Rotate=<0 1 0>, Quaternion=$Output.SocketLocation.SampledRotation)
    Sprite Facing=&RotateVectorByQuaternion(Vector To Rotate=<0 0 1>, Quaternion=$Output.SocketLocation.SampledRotation)
    WriteAlignment=true
    WriteFacing=true
   )
   /SetVariables(
    Particles.SecondAlignment:Vector3f=&RotateVectorByQuaternion(Vector To Rotate=<0 0 1>, Quaternion=$Output.SocketLocation.SampledRotation)
    Particles.SecondCameraOffset:NiagaraFloat=&Lerp_Float(A=0, B=7, Alpha=$Particles.GlowDot)
    Particles.SecondColor:LinearColor=&MultiplyLinearColor(
     Linear Color=$Particles.Initial.Color
     Scale Factor=&MakeVector4FromVectorAndFloat(
      "Vector (RGB)"=&MakeVectorFromLinearColorRGB(Color=<rgba 1 1 1 1>)
      "Float (Alpha)"=&Multiply_Float(
       A=$System.FadeIn
       B=&Lerp_Float(
        A=&Waveform(
         Global Input Position=&RandomRangeFloat(Minimum=0, Maximum=10)
         "Waveform [1]"=::NewEnumerator0
         "[1] Frequency"=1
         "[1] Amplitude Min/Max"=<0.35 0.5>
        )
        B=0
        Alpha=&Abs_Float(Float:NiagaraFloat=&DotProduct(
          Vector A=&ReturnCameraVector(Camera Vectors=::NewEnumerator1, Coordinate Space=::Local)
          Vector B=&RotateVectorByQuaternion(Vector To Rotate=<0 0 1>, Quaternion=$Output.SocketLocation.SampledRotation)
         ))
       )
      )
     )
    )
    Particles.SecondFacing:Vector3f=&RotateVectorByQuaternion(Vector To Rotate=<0 0 0>, Quaternion=$Output.SocketLocation.SampledRotation)
    Particles.SecondPosition:NiagaraPosition=$Particles.Position
    Particles.SecondSize:Vector2f=&Lerp_Vector2(
     A=<17 10>
     B=<10 10>
     Alpha=&ClampFloat(Float=&DotProduct(
       Vector A=&ReturnCameraVector(Camera Vectors=::NewEnumerator4, Coordinate Space=::Simulation, Origin Position=$Particles.Position)
       Vector B=&RotateVectorByQuaternion(Vector To Rotate=<0 1 0>, Quaternion=$Output.SocketLocation.SampledRotation)
      ))
    )
   ) {isSetParameters=true}
  ]
  rendererClasses [&NiagaraSpriteRendererProperties]
  renderers [
   {rendererClass=&NiagaraSpriteRendererProperties}
   {rendererClass=&NiagaraSpriteRendererProperties, rendererIndex=1}
  ]
 }
 Thrusters {
  emitterUpdateScript [
   &EmitterState(Life Cycle Mode=::NewEnumerator1, Loop Behavior=::NewEnumerator1, Loop Duration Mode=::NewEnumerator1, UseLoopDelay=false)
   &SpawnBurst_Instantaneous(Spawn Count=4, Use Spawn Probability=false, UseLoopCountLimit=false)
  ]
  particleSpawnScript [
   &PartitionParticles()
   &InitializeParticle(
    Color=<rgba 3 3 3 0.2>
    Lifetime=5
    Mesh Renderer Array Visibility Mode=::NewEnumerator1
    Mesh Renderer Info=DataInterface() {propertyValues {MeshRenderer=&self:Thrusters_0.NiagaraMeshRendererProperties_0}}
    Mesh Renderer Mesh Index:NiagaraInt32=&SpawnGroup()
    Mesh Scale Mode=::NewEnumerator3
    UsePositionOffset=false
    Write Material Random=true
   )
  ]
  particleUpdateScript [
   &ParticleState(Kill Particles When Lifetime Has Elapsed=false, Loop Particles Lifetime=true)
   &SocketLocation(
    SamplePosOffset=true
    Sampled Position Offset=<0 0 -2>
    Socket Index=$Particles.PartitionParticles.Partition
    Socket Reader=$System.Socket Reader
    Socket Sampling Mode=::NewEnumerator3
    VelocitySpeedLimit=false
   ) {module=SocketLocation001}
   &ScaleMeshSize(Scale Factor=&MakeVector(
     X=1
     Y=1
     Z=&Waveform(
      Waveform Count=::NewEnumerator1
      Global Input Position=&RandomRangeFloat(Minimum=0, Maximum=10)
      "Waveform [1]"=::NewEnumerator0
      "[1] Frequency"=&RandomRangeFloat(Minimum=0.5, Maximum=1.5)
      "[1] Amplitude Min/Max"=<0.9 1.1>
      "Waveform [2] Blend"=::NewEnumerator3
      "Waveform [2]"=::NewEnumerator9
      "[2] Input Position"=0
      "[2] Phase"=$Emitter.Age
      "[2] Phase Scale"=1
      "[2] Compound Frequency"=2
      "[2] Frequency"=&RandomRangeFloat(Minimum=5, Maximum=5)
      "[2] Amplitude Min/Max"=<0.5 1.2>
      "[2] Scale"=1
      "[2] Bias"=0
      "[2] Exponent"=1
      "[2] Pulse Duration"=0.1
      "[2] Clip Triangle Min/Max"=<0 0>
      "[2] Random Seed"=0
     )
    ))
   &CameraOffset(Camera Offset Amount=0.1)
   &ScaleColor(
    Scale Alpha=$System.FadeIn
    Scale RGB=&MakeVectorFromLinearColorRGB(Color=&HueShiftLinearColor(
      Hue=&RandomRangeFloat(Minimum=-0.1, Maximum=0.05)
      Saturation=1
      Value=1
      Linear Color=&ColorFromCurve(
       ColorCurve=DataInterface() {propertyValues {
         AlphaCurve {keys [<at 0 1 RCIM_Cubic>]}
         BlueCurve {keys [<at 0 1 RCIM_Cubic>,<at 1 1 RCIM_Cubic>]}
         GreenCurve {keys [<at 0 0.46 RCIM_Cubic>,<at 1 0.8259 RCIM_Cubic>]}
         RedCurve {keys [<at 0 0 RCIM_Cubic>,<at 1 0 RCIM_Cubic>]}
        }}
       CurveIndex=&Waveform(
        "Waveform [1]"=::NewEnumerator12
        "[1] Frequency"=&RandomRangeFloat(Minimum=0.2, Maximum=1)
        "[1] Amplitude Min/Max"=<1 1>
       )
      )
     ))
    ScaleA=true
    ScaleRGB=true
   )
  ]
  rendererClasses [&NiagaraMeshRendererProperties]
  renderers [{rendererClass=&NiagaraMeshRendererProperties, rendererIndex=0}]
 }
}
summary {
 systemName=NS_Backpack_FloralCardinal
 emitters [
  {emitterName=ThrusterGlow, rendererClasses [&NiagaraSpriteRendererProperties]}
  {emitterName=Thrusters, rendererClasses [&NiagaraMeshRendererProperties]}
  {emitterName=ButtonGlow, rendererClasses [&NiagaraSpriteRendererProperties, &NiagaraLightRendererProperties]}
 ]
 userVariables [
  User.bIsFrontEnd {type=NiagaraBool}
  User.bIsFrontEndPreview {type=NiagaraBool}
 ]
}
