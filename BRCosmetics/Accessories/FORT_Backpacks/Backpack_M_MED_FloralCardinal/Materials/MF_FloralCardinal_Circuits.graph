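# Material function for the FloralCardinal backpack cosmetic that adds animated, glowing circuit lines to the incoming material's emissive.
# A circuit mask, sampled with single-step parallax UVs over a flat black heightmap for apparent depth, blends two colors and masks the glow.
# Panning blurred noise modulates brightness, and an inverted, contrast-adjusted Fresnel fades it toward grazing angles.
# The result is combined with the existing emissive using Max rather than addition.
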MaterialFunction /BRCosmetics/Accessories/FORT_Backpacks/Backpack_M_MED_FloralCardinal/Materials/MF_FloralCardinal_Circuits
graph {nodes [
  n1 = GetMaterialAttributes(None=FunctionInput() {inputName=MA, inputType=FunctionInput_MaterialAttributes}) {attributeGetTypes [#1, #2]}
  n2 = TextureSampleParameter2D(UVs=MaterialFunctionCall!(
    Heightmap Texture=TextureObject() {texture=&T_BlacksRGB}
    Height Ratio=ScalarParameter() {defaultValue=0.003, group=FlaoralCardinal, parameterName=HeightRatio}
    Min Steps=Constant() {r=1}
    Max Steps=Constant() {r=1}
    UVs
    Heightmap Channel=Constant4Vector() {constant=<rgba 1 0 0 0>}.RGBA
    Reference Plane=ScalarParameter() {defaultValue=1.5, group=FlaoralCardinal, parameterName=ReferencePlane}
    -----------------------------
    Use World Coordinates
    Specify Manual Texture Size
    Manual Texture Size
    ------------------------------
    "Render Shadows (Occlusion Mapping)"
    Light Vector
    Shadow Steps
    Shadow Penumbra
    Transform To VertexNormal
   ) {materialFunction=/Engine/Functions/Engine_MaterialFunctions01/Texturing/ParallaxOcclusionMapping}.Parallax UVs) {
   group=FlaoralCardinal
   parameterName=GlowLineMask
   texture=/BRCosmetics/Accessories/FORT_Backpacks/Backpack_M_MED_FloralCardinal/Textures/T_FloralCardinal_Backpack_Circuits
  }
  n3 = LinearInterpolate(
   A=VectorParameter() {group=FlaoralCardinal, parameterName=ColorA, defaultValue=<rgba 0 0.4847 2 1>}.RGB
   B=VectorParameter() {group=FlaoralCardinal, parameterName=ColorB, defaultValue=<rgba 0.02 1.3572 2 1>}.RGB
   Alpha=@n2.R
  )
  n4 = Panner(
   Coordinate=Multiply(A=TextureCoordinate(), B=ScalarParameter() {defaultValue=5, group=FlaoralCardinal, parameterName=Noise_Tiling}) {constB=5}
   Speed=AppendVector(
    A=ScalarParameter() {defaultValue=0.02, group=FlaoralCardinal, parameterName=Noise_SpeedX}
    B=ScalarParameter() {defaultValue=-0.15, group=FlaoralCardinal, parameterName=Noise_SpeedY}
   )
  ) {speedX=0.02, speedY=-0.15}
  n5 = LinearInterpolate(
   A=ScalarParameter() {defaultValue=0.35, group=FlaoralCardinal, parameterName=Noise_Strength_Min}
   B=ScalarParameter() {defaultValue=3, group=FlaoralCardinal, parameterName=Noise_Strength_Max}
   Alpha=Power(Base=TextureSample(UVs=@n4) {samplerType=SAMPLERTYPE_Grayscale, texture=/Game/Global/Textures/Default/Noise/LowResBlurredNoiseNormalized}.R) {constExponent=1.5}
  ) {constA=0.35, constB=3}
  n6 = Fresnel(
   ExponentIn=ScalarParameter() {defaultValue=0.2067, group=FlaoralCardinal, parameterName=Fresnel_Exponentln}
   BaseReflectFractionIn=ScalarParameter() {defaultValue=0.193, group=FlaoralCardinal, parameterName=Fresnel_BaseReflectFractionln}
  )
  n7 = MaterialFunctionCall(In=OneMinus(None=Saturate(None=@n6)), Contrast=ScalarParameter() {defaultValue=1.0628, group=FlaoralCardinal, parameterName=CheapContrast}) {materialFunction=&CheapContrast}
  n8 = ScalarParameter() {defaultValue=9.0372, group=FlaoralCardinal, parameterName=CheapContrast_Blend}
  n9 = Multiply(A=Multiply(A=Power(Base=@n2.R) {constExponent=1}, B=Multiply(A=@n3, B=@n5)), B=Saturate(None=Multiply(A=@n7.Result, B=@n8)))
  FunctionOutput(None=SetMaterialAttributes!(MaterialAttributes=@n1.MaterialAttributes, EmissiveColor=Max(A=@n1.EmissiveColor, B=@n9)) {attributeSetTypes [#1]})
 ]}
